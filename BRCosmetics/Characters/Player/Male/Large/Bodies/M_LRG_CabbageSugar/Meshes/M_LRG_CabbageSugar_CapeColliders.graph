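# Physics asset of cape colliders for the large male CabbageSugar player body, shaped to keep a simulated cape off the character.
# Capsules cover the pelvis, spine, clavicles and neck, with extra capsules on spine_02 and spine_05, and a sphere covers the head.
# Arm, leg and foot bodies are present but have no shapes, so only the torso and head collide.
# It has no constraints, so it provides collision volumes rather than a jointed ragdoll.
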PhysicsAsset /BRCosmetics/Characters/Player/Male/Large/Bodies/M_LRG_CabbageSugar/Meshes/M_LRG_CabbageSugar_CapeColliders
bodies {
 calf_l []
 calf_r []
 clavicle_l [{shapeType=Capsule, length=9.0192, radius=9.8871, center=<9.4266 -1.1806 -2.4977>, rotation=<rot 69.3297 1.4476 176.9462>}]
 clavicle_r [{shapeType=Capsule, length=9.0192, radius=9.8871, center=<-9.4266 1.1806 2.4977>, rotation=<rot 69.3297 1.4476 -3.0538>}]
 foot_l []
 foot_r []
 head [{shapeType=Sphere, radius=17.3125, center=<11.028 -0.174 0>}]
 lowerarm_l []
 lowerarm_r []
 neck_02 [{shapeType=Capsule, length=13.4375, radius=7.3175, center=<-0.2642 -0.3408 0>, rotation=<rot 90 -171.1193 -175.3978>}]
 pelvis [{shapeType=Capsule, length=8.6456, radius=14.5388, center=<-2.4887 -2.7715 0>}]
 spine_01 [{shapeType=Capsule, length=7.0841, radius=14.4182, center=<-1.7965 -0.8636 0>}]
 spine_02 [
  {shapeType=Capsule, length=7.3264, radius=13.6322, center=<-1.711 -0.1851 0.0001>}
  {shapeType=Capsule, length=25.9589, radius=15.0173, center=<2.5024 -1.9562 3.9825>, rotation=<rot 90 0.0099 0.0099>}
  {shapeType=Capsule, length=26.4383, radius=14.538, center=<2.4704 -0.9736 -3.9825>, rotation=<rot 90 -0.0099 -0.0099>}
 ]
 spine_03 [{shapeType=Capsule, length=7.1249, radius=15.7589, center=<0.5825 -7.152 0.0001>, rotation=<rot 0 10.2925 0>}]
 spine_04 [{shapeType=Capsule, length=9.3816, radius=18.0949, center=<2.8724 -3.4383 0.0002>, rotation=<rot 0 7.5929 0>}]
 spine_05 [
  {shapeType=Capsule, length=8.0226, radius=15.695, center=<-2.2795 -0.6089 0>}
  {shapeType=Capsule, length=14.8869, radius=16.0569, center=<-0.2533 -4.2066 0>, rotation=<rot 0 2.0732 0>}
 ]
 thigh_l []
 thigh_r []
 upperarm_l []
 upperarm_r []
}
constraints []
